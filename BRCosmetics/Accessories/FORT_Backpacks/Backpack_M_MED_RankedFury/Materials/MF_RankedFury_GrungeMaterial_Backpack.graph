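# Material function for the Ranked Fury backpack that adds grunge to incoming material attributes, blended in by a TechArtMask input.
# A grunge texture, replaced by a constant color on low/mobile, darkens or modulates BaseColor and optionally multiplies into offset Roughness.
# The grunged color is desaturated and hue-shifted by parameters.
# An optional pre-skinning local-space sphere mask exempts the umbrella top from that color shift.
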MaterialFunction /BRCosmetics/Accessories/FORT_Backpacks/Backpack_M_MED_RankedFury/Materials/MF_RankedFury_GrungeMaterial_Backpack
graph {nodes [
  n1 = GetMaterialAttributes(None=FunctionInput() {inputName=MaterialAttributes, inputType=FunctionInput_MaterialAttributes}) {attributeGetTypes [#1, #2]}
  n2 = Add(A=@n1.Roughness, B=ScalarParameter() {defaultValue=0.15, group=ThinFilm_RankedFury, parameterName=OverallRoughnessOffset, sortPriority=30})
  n3 = TextureSample() {
   samplerSource=SSM_Wrap_WorldGroupSettings
   texture=/BRCosmetics/Accessories/FORT_Backpacks/Backpack_M_MED_RankedFury/Textures/T_RankedFury_Backpack_Grunge
  }
  n4 = Power(
   Base=MaterialFunctionCall!(Default=@n3.RGB, Low/Mobile=Constant3Vector() {constant=<rgba 0.3968 0.3564 0.3185 1>}) {materialFunction=/Game/Effects/Fort_Effects/Materials/_Functions/MF_LowMobile_PathingSwitch}.Result
   Exp=ScalarParameter() {defaultValue=1, group=ThinFilm_RankedFury, parameterName=GrungePower, sortPriority=30}
  )
  n5 = Add(A=@n4, B=ScalarParameter() {defaultValue=0.15, group=ThinFilm_RankedFury, parameterName=GrungeRoughnessOffset, sortPriority=30})
  n6 = StaticSwitchParameter(True=Multiply(A=Saturate(None=@n2), B=Saturate(None=@n5)), False=Saturate(None=Add(A=@n1.Roughness) {constB=0.025})) {group=ThinFilm_RankedFury, parameterName=AddGrungeToRoughness, sortPriority=0}
  n7 = ScalarParameter() {defaultValue=0.15, group=ThinFilm_RankedFury, parameterName=GrungeOffset, sortPriority=30}
  n8 = Multiply(A=@n1.BaseColor, B=Saturate(None=Add(A=@n4, B=@n7) {constB=0.15}))
  n9 = MaterialFunctionCall(
   Hue Shift Percentage=ScalarParameter() {group=ThinFilm_RankedFury, parameterName=BaseColorHueShift}
   Texture=Desaturation(None=@n8, Fraction=ScalarParameter() {group=ThinFilm_RankedFury, parameterName=BaseColorDesaturation})
  ) {materialFunction=&HueShift}
  n10 = LocalPosition() {localOrigin=InstancePreSkinning}
  n11 = Multiply(A=SphereMask(A=VertexInterpolator(VS=@n10.XYZ).PS, B=Constant3Vector() {constant=<rgba 0 0 130 0>}) {attenuationRadius=25, hardnessPercent=25}) {constB=2}
  n12 = StaticSwitchParameter(True=LinearInterpolate(A=@n9.Result, B=@n8, Alpha=Saturate(None=@n11)), False=@n9.Result) {group=ThinFilm_RankedFury, parameterName=UseUmbrellaTopMask, sortPriority=0}
  n13 = FunctionInput() {bUsePreviewValueAsDefault=true, inputName=TechArtMask, inputType=FunctionInput_Scalar, sortPriority=1, previewValue=<1 0 0 1>}
  FunctionOutput(None=BlendMaterialAttributes(
    A=@n1.MaterialAttributes
    B=SetMaterialAttributes!(MaterialAttributes=@n1.MaterialAttributes, Roughness=@n6, BaseColor=@n12) {attributeSetTypes [#1, #2]}
    Alpha=@n13
   ))
 ]}
